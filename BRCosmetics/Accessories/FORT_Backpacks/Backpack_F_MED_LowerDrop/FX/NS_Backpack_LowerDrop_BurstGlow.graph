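# Niagara burst effect for the LowerDrop backpack cosmetic: a warm glow flash with sparkles.
# Glow bursts into three sprite renderers, with separate flare and corona attributes animated by curves alongside base size and color.
# Sparkles uses two spawn groups, switching lifetime, speed, color and size per group: quick, short-lived sparks and slower, lingering ones that twinkle through a sine.
# Both emitters tint their colors with a system time-of-day multiplier.
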NiagaraSystem /BRCosmetics/Accessories/FORT_Backpacks/Backpack_F_MED_LowerDrop/FX/NS_Backpack_LowerDrop_BurstGlow
dependencies {
 usedDataInterfaces [
  &NiagaraDataInterfaceCurve
  &NiagaraDataInterfaceColorCurve
  &NiagaraDataInterfaceCamera
  &NiagaraDataInterfaceSpriteRendererInfo
 ]
 usedDynamicInputs [&RandomRangeVector2D]
 usedModules [
  &self:SystemSpawnScript.SystemScriptSource.SystemScriptGraph.NiagaraNodeAssignment_0.SetVariables_#1
  &SystemState
 ]
}
emitters {
 Glow {
  emitterUpdateScript [
   &EmitterState()
   &SpawnBurst_Instantaneous(Use Spawn Probability=false, UseLoopCountLimit=false)
  ]
  particleSpawnScript [
   &InitializeParticle(
    Color=<rgba 0.5104 0.0263 0.0416 1>
    Position Offset=<0 -2.5 2.5>
    Position Offset Coordinate Space=::Local
    Sprite Size Mode=::NewEnumerator3
    Uniform Sprite Size=75
    UsePositionOffset=true
    Write Material Random=true
   )
   /SetVariables(
    Particles.FlareCameraOffset=10
    Particles.FlareColor=&MultiplyLinearColor(
     Linear Color=<rgba 5 3.1 0.25 0.5>
     Scale Factor=&MakeVector4_FromLinearColor(LinearColor=$System.TODColorMultiplier)
    )
    Particles.FlareSpriteSize:Vector2f=&Multiply_Vector2D(A=$Particles.SpriteSize, B=<1 1>)
   ) {isSetParameters=true}
   /SetVariables(
    Particles.CoronaColor:LinearColor=&MultiplyLinearColor(
     Linear Color=<rgba 3 2.4411 1.2713 0.45>
     Scale Factor=&MakeVector4_FromLinearColor(LinearColor=$System.TODColorMultiplier)
    )
    Particles.CoronaSize:Vector2f=&Vector2DFromFloat(Value=45)
   ) {isSetParameters=true}
   &CameraOffset(Camera Offset Amount=5)
  ]
  particleUpdateScript [
   &ParticleState()
   &ScaleSpriteSize(Uniform Curve Sprite Scale=DataInterface() {propertyValues {Curve {keys [{interpMode=RCIM_Cubic, leaveTangent=4.1281, tangentMode=RCTM_Break, time=0, value=0}, {tangentMode=RCTM_User, time=0.5, value=0.95}]}}})
   &ScaleColor(
    Linear Color Curve=DataInterface() {propertyValues {
      AlphaCurve {keys [<at 0 1>,<at 0.2675 0.4>,<at 0.7143 0>]}
      BlueCurve {keys [<at 0 0>,<at 0.1018 1>,<at 0.6368 1>,<at 1 0>]}
      GreenCurve {keys [<at 0 0.4292>,<at 0.1018 1>,<at 0.6368 1>,<at 1 0.3697>]}
      RedCurve {keys [<at 0 2>,<at 0.1018 1>,<at 0.6368 1>,<at 1 1>]}
     }}
    Scale Mode=::NewEnumerator2
   )
   /SetVariables(Particles.FlareColor=&MultiplyLinearColor(
     Linear Color=$Particles.FlareColor
     Scale Factor=&MakeVector4_FromLinearColor(LinearColor=&ColorFromCurve(ColorCurve=DataInterface() {propertyValues {
         AlphaCurve {keys [<at 0 1 RCIM_Cubic>,<at 0.701 0 RCIM_Cubic>]}
         BlueCurve {keys [<at 0 1 RCIM_Cubic>,<at 1 1 RCIM_Cubic>]}
         GreenCurve {keys [<at 0 1 RCIM_Cubic>,<at 1 1 RCIM_Cubic>]}
         RedCurve {keys [<at 0 1 RCIM_Cubic>,<at 1 1 RCIM_Cubic>]}
        }}))
    )) {isSetParameters=true}
   /SetVariables(
    Particles.CoronaColor:LinearColor=&MultiplyLinearColor(
     Linear Color=$Particles.Initial.CoronaColor
     Scale Factor=&Vector4FromCurve(Vector4Curve=DataInterface() {propertyValues {
        AlphaCurve {keys [<at 0 1>,<at 0.3702 0.2>,<at 1 0>]}
        BlueCurve {keys [<at 0 0>,<at 0.1018 1>,<at 0.6368 1>,<at 1 0>]}
        GreenCurve {keys [<at 0 5.2558>,<at 0.1018 1>,<at 0.6368 1>,<at 1 0.3697>]}
        RedCurve {keys [<at 0 10>,<at 0.1018 1>,<at 0.6368 1>,<at 1 1>]}
       }})
    )
    Particles.CoronaSize:Vector2f=&Multiply_Vector2D(
     A=$Particles.Initial.CoronaSize
     B=&Vector2DFromFloat(Value=&FloatFromCurve(FloatCurve=DataInterface() {propertyValues {Curve {keys [
           {interpMode=RCIM_Cubic, leaveTangent=4.6427, tangentMode=RCTM_Break, time=0, value=0}
           {arriveTangent=0.7668, tangentMode=RCTM_User, time=0.7, value=0.95}
          ]}}}))
    )
   ) {isSetParameters=true}
   &ScaleColor(Color Scale:LinearColor=$System.TODColorMultiplier, Scale Mode=3) {module=ScaleColor001}
  ]
  rendererClasses [&NiagaraSpriteRendererProperties]
  renderers [
   {rendererClass=&NiagaraSpriteRendererProperties}
   {rendererClass=&NiagaraSpriteRendererProperties, rendererIndex=1}
   {rendererClass=&NiagaraSpriteRendererProperties, rendererIndex=2}
  ]
 }
 Sparkles {
  emitterUpdateScript [
   &EmitterState()
   &SpawnBurst_Instantaneous(
    Spawn Count=&RandomRangeInt(Minimum=5, Maximum=8)
    Spawn Time=0.05
    Use Spawn Probability=false
    UseLoopCountLimit=false
   )
   &SpawnBurst_Instantaneous(
    Spawn Count=&RandomRangeInt(Minimum=8, Maximum=12)
    Spawn Group=1
    Use Spawn Probability=false
    UseLoopCountLimit=false
   ) {module=SpawnBurst_Instantaneous001}
  ]
  particleSpawnScript [
   &InitializeParticle(
    Color=<rgba 10 6.4776 0.7081 0.8>
    Lifetime=&MergeHelper_FloatSwitch(
     0=0.4
     1=&RandomRangeFloat(Minimum=1.5, Maximum=2.5)
     2=?
     3=?
     Index=&SpawnGroup()
     Num=::NewEnumerator0
    )
    Position Offset=<0 0 2.5>
    Sprite Rotation Angle=45
    Sprite Rotation Mode=::NewEnumerator0
    Sprite Size Mode=::NewEnumerator1
    Uniform Sprite Size Max=7
    Uniform Sprite Size Min=3
    UsePositionOffset=true
    Write Material Random=true
   )
   &ShapeLocation(
    MaskSpawnGroup=false
    Rotation Angle:NiagaraFloat=&RandomRangeFloat(Minimum=0, Maximum=45)
    Rotation Angle Type:ENiagara_AngleInput=::NewEnumerator0
    Rotation Axis=&ReturnCameraVector(Camera Vectors=::NewEnumerator1, Coordinate Space=::Simulation)
    Rotation Mode=::NewEnumerator1
    Sphere Distribution=::NewEnumerator3
    Sphere Radius=&MergeHelper_FloatSwitch(0=8, 1=6, 2=?, 3=?, Index=&SpawnGroup(), Num=::NewEnumerator0)
    UseOwnerScale=true
   )
   &AddVelocity(
    Module.Use Falloff Distance=false
    Override Seed=false
    Velocity Mode=::NewEnumerator1
    Velocity Speed=&MergeHelper_FloatSwitch(
     0=&RandomRangeFloat(Minimum=75, Maximum=50)
     1=&RandomRangeFloat(Minimum=20, Maximum=10)
     2=?
     3=?
     Index=&SpawnGroup()
     Num=::NewEnumerator0
    )
   )
   &SubUVAnimation(
    Direct Frame Index:NiagaraFloat=&RandomRangeFloat(Minimum=0, Maximum=1)
    "Sprite Renderer "=DataInterface() {propertyValues {SpriteRenderer=&self:Sparkles_1.NiagaraSpriteRendererProperties_0}}
    SubUV Animation Mode:ENiagara_SubUVLookupModeV2=::NewEnumerator5
   )
  ]
  particleUpdateScript [
   &ParticleState()
   &Drag(Drag=5, Use Linear Drag=true, Use Rotational Drag=true)
   &GravityForce(Gravity=<0 0 -2>)
   &SolveForcesAndVelocity(Clamp Velocity=false, Limit Acceleration=false)
   &ScaleColor(
    Scale Mode=::NewEnumerator1
    Scale RGBA=&MakeVector4_FromLinearColor(LinearColor=&MultiplyLinearColor(
      Linear Color=$System.TODColorMultiplier
      Scale Factor=&MakeVector4_FromLinearColor(LinearColor=&MergeHelper_ColorSwitch(
        0=&ColorFromCurve(ColorCurve=DataInterface() {propertyValues {
           AlphaCurve {keys [<at 0.0984 1>,<at 1 0>]}
           BlueCurve {keys [<at 0 0>,<at 0.5754 1>,<at 0.7508 1>,<at 1 0>]}
           GreenCurve {keys [<at 0 5.2558>,<at 0.5754 1>,<at 0.7508 1>,<at 1 0.3697>]}
           RedCurve {keys [<at 0 10>,<at 0.5754 1>,<at 0.7508 1>,<at 1 1>]}
          }})
        1=&ColorFromCurve(ColorCurve=DataInterface() {propertyValues {
           AlphaCurve {keys [<at 0 0>,<at 0.2534 0.4>,<at 0.8752 0.2357>,<at 1 0>]}
           BlueCurve {keys [<at 0 1>,<at 1 0>]}
           GreenCurve {keys [<at 0 1>,<at 1 0.3697>]}
           RedCurve {keys [<at 0 1>,<at 1 1>]}
          }})
        2=?
        3=?
        Index=&SpawnGroup()
        Num=::NewEnumerator0
       ))
     ))
    ScaleRGBA=true
   )
   /SetVariables(Particles.SpriteSize=&MergeHelper_Vector2Switch(
     0=&Multiply_Vector2D_ByFloat(
      Float=&FloatFromCurve(FloatCurve=DataInterface() {propertyValues {Curve {
          defaultValue=0
          keys [
           {arriveTangent=14.7451, interpMode=RCIM_Cubic, leaveTangent=14.7451, tangentMode=RCTM_User, time=0, value=0}
           {arriveTangent=-0.0064, interpMode=RCIM_Cubic, leaveTangent=-0.0064, tangentMode=RCTM_User, time=0.1408, value=0.9847}
           {arriveTangent=0.0962, interpMode=RCIM_Cubic, leaveTangent=0.0962, tangentMode=RCTM_User, time=1, value=1}
          ]
         }}})
      Vector2D=$Particles.Initial.SpriteSize
     )
     1=&Multiply_Vector2D_ByFloat(
      Float=&Multiply_Float(
       A=&FloatFromCurve(FloatCurve=DataInterface() {propertyValues {Curve {
           defaultValue=0
           keys [
            {arriveTangent=0.3364, interpMode=RCIM_Cubic, leaveTangent=0.3364, tangentMode=RCTM_User, time=0, value=0}
            {interpMode=RCIM_Cubic, tangentMode=RCTM_User, time=1, value=1}
           ]
          }}})
       B=&Sine(
        Normalized Angle=$Particles.NormalizedAge
        Period=&RandomRangeFloat(Minimum=0.45, Maximum=0.8)
        Scale=1
        Bias=&RandomRangeFloat(Minimum=0, Maximum=1)
       )
      )
      Vector2D=$Particles.Initial.SpriteSize
     )
     2=?
     3=?
     4=?
     Index=&SpawnGroup()
     Num=::NewEnumerator0
    )) {isSetParameters=true}
  ]
  rendererClasses [&NiagaraSpriteRendererProperties]
  renderers [{rendererClass=&NiagaraSpriteRendererProperties}]
 }
}
summary {
 systemName=NS_Backpack_LowerDrop_BurstGlow
 emitters [
  {emitterName=Glow, rendererClasses [&NiagaraSpriteRendererProperties]}
  {emitterName=Sparkles, rendererClasses [&NiagaraSpriteRendererProperties]}
 ]
 userVariables []
}
