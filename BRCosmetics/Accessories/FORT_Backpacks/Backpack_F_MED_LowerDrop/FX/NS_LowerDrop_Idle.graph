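# Idle glow effect for the LowerDrop backpack cosmetic, built from a single emitter with three sprite renderers.
# A burst spawns three persistent particles, partitioned so each gets its own size, rotation, alignment and facing at a small system offset.
# A per-partition waveform pulses alpha, scaled by a front-end versus gameplay modifier, with a camera offset applied.
# User.AnimElimKill feeds a dynamic material parameter, and the particles can die when the emitter deactivates.
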NiagaraSystem /BRCosmetics/Accessories/FORT_Backpacks/Backpack_F_MED_LowerDrop/FX/NS_LowerDrop_Idle
dependencies {usedDataInterfaces [], usedDynamicInputs [], usedModules [&SystemState]}
emitters {Glow {
  emitterUpdateScript [
   &EmitterState(Life Cycle Mode=::NewEnumerator1, Loop Behavior=::NewEnumerator1, Loop Duration Mode=::NewEnumerator1, UseLoopDelay=false)
   &SpawnBurst_Instantaneous(Spawn Count=3, Use Spawn Probability=false, UseLoopCountLimit=false)
  ]
  particleSpawnScript [
   &PartitionParticles(Number Of Partitions=3)
   &InitializeParticle(
    Sprite Rotation Angle=&MergeHelper_FloatSwitch(0=0, 1=0, 2=90, 3=?, Index=$Particles.PartitionParticles.Partition, Num=::NewEnumerator1)
    Sprite Rotation Mode=::NewEnumerator0
    Sprite Size=&MergeHelper_Vector2Switch(0=<17.5 37.5>, 1=<27.5 27.5>, 2=<10 20>, 3=?, 4=?, Index=$Particles.PartitionParticles.Partition, Num=::NewEnumerator1)
    Sprite Size Mode=::NewEnumerator4
    UsePositionOffset=false
    Write Material Random=true
   )
   &SystemLocation(MaskSpawnGroup=false, Offset=<0 -3 3.5>)
   /Game/Sprite/FN_SpriteAlignmentAndFacing(
    SpriteAlignment:Vector3f=&MergeHelper_VectorSwitch(0=<1 0 0>, 1=<1 0 0>, 2=<0 1 0>, 3=?, Index=$Particles.PartitionParticles.Partition, Num=::NewEnumerator1)
    SpriteFacing:Vector3f=&MergeHelper_VectorSwitch(0=<0 1 0>, 1=<0 1 0>, 2=<0 0 0>, 3=?, Index=$Particles.PartitionParticles.Partition, Num=::NewEnumerator1)
   )
  ]
  particleUpdateScript [
   &ParticleState(Kill Particles When Lifetime Has Elapsed=false, Let Infinitely Lived Particles Die When Emitter Deactivates=true)
   &CameraOffset(Camera Offset Amount=5)
   &ScaleColor(
    Scale Alpha=&Multiply_Float(
     A=&FN_FrontEndFloatModifier(Gameplay=1, Front End=0.75)
     B=&Waveform(
      Global Amplitude Scale=&MergeHelper_FloatSwitch(0=2, 1=0.175, 2=3, 3=?, Index=$Particles.PartitionParticles.Partition, Num=::NewEnumerator1)
      "Waveform [1]"=::NewEnumerator12
      "[1] Frequency"=1
      "[1] Amplitude Min/Max"=<0.5 0.75>
     )
    )
    ScaleA=true
    ScaleRGB=true
   )
   &DynamicMaterialParameters(
    Index 0 Param 1=$User.AnimElimKill
    Module.Index0Float0Write=true
    Module.Index0Float1Write=false
    Module.Index0Float2Write=false
    Module.Index0Float3Write=false
   )
  ]
  rendererClasses [&NiagaraSpriteRendererProperties]
  renderers [
   {rendererClass=&NiagaraSpriteRendererProperties}
   {rendererClass=&NiagaraSpriteRendererProperties, rendererIndex=1}
   {rendererClass=&NiagaraSpriteRendererProperties, rendererIndex=2}
  ]
 }}
summary {
 systemName=NS_LowerDrop_Idle
 emitters [{emitterName=Glow, rendererClasses [&NiagaraSpriteRendererProperties]}]
 userVariables [
  User.AnimElimKill {type=NiagaraFloat, defaultValue=0, description=""}
  User.bIsFrontEnd {type=NiagaraBool}
  User.bIsFrontEndPreview {type=NiagaraBool}
 ]
}
