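# Material function for the LowerDrop backpack cosmetic that overrides base color, metallic, specular, roughness, normal and emissive on incoming material attributes.
# Projected, rotated UVs pan a mask texture that, with a sine offset, indexes a gradient texture for animated color.
# A Golden Animation Blend parameter, noise and a packed curve mask build a golden-style mask and glowing edge via sphere masks.
# Logo, text, AO and thickness masks plus Fresnel supply further shading detail.
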MaterialFunction /BRCosmetics/Accessories/FORT_Backpacks/Backpack_F_MED_LowerDrop/Material/MF_Backpack_LowerDrop
graph {nodes [
  n1 = SetMaterialAttributes!(
   MaterialAttributes=GetMaterialAttributes(None=FunctionInput() {bUsePreviewValueAsDefault=true, inputType=FunctionInput_MaterialAttributes}).MaterialAttributes
   BaseColor=NamedRerouteUsage()
   Metallic=Constant() {r=0.2}
   Specular=Constant() {r=0.5}
   Roughness=NamedRerouteDeclaration(None=Constant() {r=0.3}) {name=Roughness, nodeColor=<rgba 0 0.0859 1 1>}
   Normal=NamedRerouteUsage()
   EmissiveColor=NamedRerouteUsage()
  ) {attributeSetTypes [
    #1
    #2
    #3
    #4
    #5
    #6
   ]}
  FunctionOutput(None=@n1)
  n2 = Constant()
  n3 = Constant() {r=-1.1446}
  n4 = Constant()
  n5 = Constant() {r=1.3401}
  n6 = MaterialFunctionCall!(Root_Position=Constant(), UseClamp?, FlipUV, UVSize=Constant() {r=30}, CenterOffsetX=@n3, CenterOffsetY=@n4, CenterOffsetZ=@n5) {materialFunction=/BRCosmetics/Gadgets/Assets/VinderTech_GliderChute/Glider_WiryPerk/Materials/MF_ProjectUV_WiryPerk}
  n7 = Subtract(
   A=VertexInterpolator(VS=LocalPosition() {localOrigin=InstancePreSkinning}.XYZ).PS
   B=MaterialFunctionCall(R=@n3, G=@n4, B=@n5) {materialFunction=&AppendMany}.RGB
  )
  n8 = Constant() {r=30}
  n9 = ComponentMask(None=ScalarParameter() {defaultValue=1, group=LowerDrop, parameterName=EmissivePannerScale}) {
   desc="Kendall added this node on 5/27/2020.\r\nPreviously, this just piped the 3-vector value from\r\nthe texture sample straight into the UV's, which\r\ncaused the material to fail compile"
   g=true
   r=true
  }
  n10 = Multiply(
   A=MaterialFunctionCall(
    UVs=NamedRerouteDeclaration(None=@n6.Result) {name=FrontProjectUV, nodeColor=<rgba 0.2422 0 1 1>}
    Rotation Center=Divide(A=@n7, B=@n8)
    "Rotation Angle (0-1)"=Constant() {r=0.104}
   ) {materialFunction=&CustomRotator}.Rotated Values
   B=@n9
  )
  n11 = ScalarParameter() {defaultValue=0.05, group=LowerDrop, parameterName=GradientSpeed_X, sortPriority=2}
  n12 = ScalarParameter() {defaultValue=0.15, group=LowerDrop, parameterName=GradientSpeed_Y, sortPriority=2}
  n13 = TextureObjectParameter() {
   group=06_Advanced Emissive
   parameterName=EmissiveGradient
   samplerType=SAMPLERTYPE_Masks
   sortPriority=1
   texture=/Game/Global/Textures/Default/Masks_and_Gradients/T_VertGradient_MASK
  }
  n14 = TextureSample(UVs=Panner(Coordinate=Add(A=@n2, B=@n10), Speed=AppendVector(A=@n11, B=@n12)), Tex=@n13) {
   samplerSource=SSM_Wrap_WorldGroupSettings
   samplerType=SAMPLERTYPE_Masks
   texture=/Game/Weapons/WeaponSkins/Textures/T_Swirly_Noise_highContrast
  }
  n15 = Time()
  n16 = Sine(None=Multiply(A=Multiply(A=@n15) {constB=0.1}, B=ScalarParameter() {group=LowerDrop, parameterName=SineSpeed, sortPriority=2}))
  n17 = Add(
   A=@n14.R
   B=LinearInterpolate(B=ScalarParameter() {defaultValue=0.5, group=LowerDrop, parameterName=ColorOffset}, Alpha=ConstantBiasScale(None=@n16)) {constB=0.5}
  )
  n18 = Constant() {r=0.3008}
  n19 = TextureSampleParameter2D(UVs=ComponentMask(None=Multiply(A=@n17, B=@n18) {constB=0.5}) {g=true, r=true}) {
   group=Quail Wink / Loan Float / Evoke Find
   parameterName=Gradient Color
   texture=&/T_Backpack_LowerDrop_Gradient
  }
  n20 = Multiply(
   A=VectorParameter() {group=LowerDrop, parameterName=GoldenColor, defaultValue=<rgba 1.8 0.9405 0.45 1>}.RGB
   B=Desaturation(None=@n19.RGB, Fraction=Constant() {r=1})
  )
  n21 = Saturate(None=Multiply(A=NamedRerouteUsage(), B=NamedRerouteUsage()))
  NamedRerouteDeclaration(None=Desaturation(
    None=LinearInterpolate(A=@n19.RGB, B=@n20, Alpha=@n21)
    Fraction=ScalarParameter() {defaultValue=-0.5, group=LowerDrop, parameterName=ColorDesaturation}
   )) {name=Gradient Color, nodeColor=<rgba 1 0 0.8906 1>}
  n22 = ScalarParameter() {group=LowerDrop, parameterName=Golden Animation Blend}
  n23 = OneMinus(None=Subtract(A=@n22))
  n24 = Add(A=NamedRerouteDeclaration(None=If(A=@n22, "A > B"=@n23, "A == B"=@n22, "A < B"=@n22) {constB=1}) {name=Golden Animation Blend, nodeColor=<rgba 1 0 0.7266 1>}) {constB=2}
  n25 = TextureSample(UVs=NamedRerouteUsage()) {
   samplerSource=SSM_Wrap_WorldGroupSettings
   samplerType=SAMPLERTYPE_Grayscale
   texture=/Game/Global/Textures/Default/Noise/LowResBlurredNoiseNormalized
  }
  n26 = TextureSample() {samplerType=SAMPLERTYPE_Masks, texture=/BRCosmetics/Accessories/FORT_Backpacks/Backpack_F_MED_LowerDrop/Texture/T_Backpack_LowerDrop_FX1}
  n27 = Multiply(
   A=Saturate(None=Multiply(A=@n24, B=@n25.R) {constB=2})
   B=Add(A=NamedRerouteDeclaration(None=@n26.R) {name=Curve, nodeColor=<rgba 1 0 0.5391 1>}) {constB=0.1}
  )
  n28 = Saturate(None=SphereMask(A=@n27, B=Constant() {r=1}, Radius=NamedRerouteUsage()))
  n29 = Saturate(None=Add(A=@n27, B=@n28))
  NamedRerouteDeclaration(None=If(A=NamedRerouteUsage(), "A > B"=@n29, "A == B"=Constant(), "A < B"=@n29)) {name=GoldenStyle, nodeColor=<rgba 0.1953 1 0 1>}
  n30 = SphereMask(A=@n27, B=Constant() {r=1}, Radius=Subtract(A=NamedRerouteUsage(), B=Constant() {r=0.05}) {constB=0.2})
  n31 = Multiply(
   A=Subtract(A=@n28, B=Saturate(None=@n30))
   B=VectorParameter() {group=LowerDrop, parameterName=GlowEdgeColor, defaultValue=<rgba 5 1.8835 0 1>}.RGB
  )
  n32 = NamedRerouteUsage()
  n33 = OneMinus(None=If(A=@n32, "A > B"=@n32, "A == B"=Constant() {r=1}, "A < B"=@n32))
  NamedRerouteDeclaration(None=Saturate(None=Multiply(A=@n31, B=@n33))) {name=GlowEdge, nodeColor=<rgba 1 0.1641 0 1>}
  n34 = Constant() {r=0.1}
  n35 = Constant() {r=0.2}
  n36 = NamedRerouteUsage()
  n37 = LinearInterpolate(
   A=NamedRerouteUsage()
   B=Multiply(A=NamedRerouteUsage(), B=LinearInterpolate(A=@n34, B=@n35, Alpha=@n36) {constB=0.2})
   Alpha=Subtract(A=NamedRerouteUsage(), B=NamedRerouteUsage())
  )
  n38 = Constant() {r=1}
  n39 = NamedRerouteUsage()
  n40 = LinearInterpolate(
   A=LinearInterpolate(
    A=Constant4Vector() {constant=<rgba 0 1.3069 2 1>}.RGB
    B=Constant4Vector() {constant=<rgba 2.8053 0.7182 3 1>}.RGB
    Alpha=NamedRerouteUsage()
   )
   B=VectorParameter() {group=LowerDrop, parameterName=LogoGoldenColor, defaultValue=<rgba 2 1.3855 0.487 1>}.RGB
   Alpha=NamedRerouteUsage()
  )
  n41 = NamedRerouteUsage()
  n42 = Add(
   A=NamedRerouteUsage()
   B=LinearInterpolate(A=Multiply(A=@n37, B=@n38), B=LinearInterpolate(A=@n39, B=@n40, Alpha=@n41) {constAlpha=0.85}, Alpha=NamedRerouteUsage())
  )
  NamedRerouteDeclaration(None=Multiply(A=NamedRerouteDeclaration(None=@n42) {name=Gradient Texture, nodeColor=<rgba 0.6172 1 0 1>}, B=Constant() {r=0.3}) {constB=0.75}) {name=BaseColor, nodeColor=<rgba 0 1 0.5547 1>}
  n43 = LinearInterpolate(A=Constant() {r=-0.232}, B=Constant() {r=1}, Alpha=NamedRerouteUsage())
  n44 = Fresnel() {exponent=1}
  n45 = LinearInterpolate(Alpha=NamedRerouteUsage()) {constA=0.5}
  n46 = TextureSample() {samplerType=SAMPLERTYPE_Masks, texture=/BRCosmetics/Accessories/FORT_Backpacks/Backpack_F_MED_LowerDrop/Texture/T_Backpack_LowerDrop_FX2}
  n47 = Constant() {r=2}
  n48 = NamedRerouteDeclaration(None=Power(Base=@n46.R, Exp=@n47) {constExponent=2.2}) {name=Ao, nodeColor=<rgba 0.1719 0 1 1>}
  n49 = LinearInterpolate(A=LinearInterpolate(A=Constant() {r=1}, B=Constant() {r=0.55}, Alpha=NamedRerouteUsage()), Alpha=NamedRerouteUsage()) {constA=1}
  n50 = Multiply(A=NamedRerouteUsage(), B=Add(A=Multiply(A=@n44, B=@n45), B=Power(Base=@n48, Exp=@n49) {constExponent=4}))
  n51 = LinearInterpolate(
   A=LinearInterpolate(A=Constant() {r=2}, B=Constant() {r=4}, Alpha=NamedRerouteUsage()) {constA=1.1}
   B=Constant() {r=4}
   Alpha=NamedRerouteUsage()
  )
  NamedRerouteDeclaration(None=Multiply(A=Saturate(None=@n43), B=Multiply(A=@n50, B=@n51))) {name=EmissiveColor, nodeColor=<rgba 0.4062 1 0 1>}
  n52 = TextureSample() {samplerType=SAMPLERTYPE_Normal, texture=/BRCosmetics/Accessories/FORT_Backpacks/Backpack_F_MED_LowerDrop/Texture/T_Backpack_LowerDrop_normal}
  NamedRerouteDeclaration(None=@n52.RGB) {name=Normal, nodeColor=<rgba 1 0 0.7266 1>}
  n53 = TextureCoordinate()
  n54 = Constant() {r=2}
  n55 = TextureSample(UVs=Panner(Coordinate=Multiply(A=@n53, B=@n54)) {speedY=0.03}) {
   samplerSource=SSM_Wrap_WorldGroupSettings
   samplerType=SAMPLERTYPE_Grayscale
   texture=&LowResBlurredNoiseNormalized
  }
  n56 = Saturate(None=Power(Base=Multiply(A=@n55.R) {constB=1.5}) {constExponent=1.2})
  n57 = OneMinus(None=LinearInterpolate(Alpha=NamedRerouteDeclaration(None=@n56) {name=ColorNoise, nodeColor=<rgba 1 0.375 0 1>}) {constB=0.5})
  n58 = Constant()
  n59 = Constant() {r=16}
  n60 = Constant()
  n61 = Constant()
  n62 = Constant() {r=2.6119}
  n63 = Multiply(
   A=TextureSample(UVs=MaterialFunctionCall(Root_Position=@n58, UVSize=@n59, CenterOffsetX=@n60, CenterOffsetY=@n61, CenterOffsetZ=@n62) {materialFunction=&MF_ProjectUV_WiryPerk}.Result) {samplerType=SAMPLERTYPE_LinearGrayscale, texture=/Game/Global/Textures/Default/Masks_and_Gradients/T_ScreenRamp_H}.R
   B=LinearInterpolate(A=Constant() {r=1.8}, B=Constant() {r=1.2}, Alpha=NamedRerouteUsage()) {constB=2}
  )
  NamedRerouteDeclaration(None=Saturate(None=Multiply(A=@n57, B=@n63))) {name=LogoMask, nodeColor=<rgba 0.0547 0 1 1>}
  NamedRerouteDeclaration(None=@n26.G) {name=TextMask, nodeColor=<rgba 0 0.8594 1 1>}
  NamedRerouteDeclaration(None=@n26.B) {name=LogoMask_1, nodeColor=<rgba 1 0.8203 0 1>}
  NamedRerouteDeclaration(None=Power(Base=@n46.G, Exp=@n47) {constExponent=2.2}) {name=Thickness, nodeColor=<rgba 0.1484 1 0 1>}
 ]}
